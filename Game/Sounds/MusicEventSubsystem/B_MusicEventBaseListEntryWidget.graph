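# Base UserWidget for a text list entry tied to the music event debug widget, showing one outlined text block in a fixed-width SizeBox.
# Every Tick polls GetFormatText and GetTextColor to set the text and color; both are empty stubs here, apparently meant for child overrides.
# When the DebugWidget reference is valid, Tick also copies its FontSize into the text's font.
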WidgetBlueprint /Game/Sounds/MusicEventSubsystem/B_MusicEventBaseListEntryWidget
graphs [
 EventGraph {nodes [
   Event Tick()
   n1 = Get Text!() -> (Text:Text Block Object Reference)
   > "SetText (Text)"(self=@n1, InText=GetFormatText!() -> (Output:Text))
   > SetColorAndOpacity(self=@n1, InColorAndOpacity=GetTextColor!() -> (Output:Slate Color Structure))
   > n2 = Get!(execute:Exec) -> (then:Exec, else:Exec, DebugWidget:B Music Event Debug Widget Object Reference)
   > n3 = Set members in Slate Font Info!(
    execute:Exec
    StructRef:"Slate Font Info Structure (by ref)"=Get Font!(self:Text Object Reference=Get Text()) -> (Font:Slate Font Info Structure)
    Size:float="To Float (Integer)"(InInt=Get FontSize!(self:B Music Event Debug Widget Object Reference=@n2) -> (FontSize:int=0))
   ) -> (then:Exec, StructOut:"Slate Font Info Structure (by ref)") {is=title K2Node_SetFieldsInStruct}
   > SetFont(self=Get Text(), InFontInfo=@n3)
  ]}
 GetFormatText {nodes [
   GetFormatText() {is=title K2Node_FunctionEntry}
   > Return Node!(execute, Output:Text)
  ]}
 GetTextColor {nodes [
   GetTextColor() {is=title K2Node_FunctionEntry}
   > Return Node(Output:Slate Color Structure)
  ]}
]
variables [DebugWidget {type=B_MusicEventDebugWidget_C, default=null}]
widgets {
 info {inheritedWidgetCount=0, parentClass=&UserWidget, rootWidgetClass=&SizeBox, widgetCount=2}
 widgets [
  SizeBox_88 {widgetClassPath=&SizeBox, widget=SizeBox_88, properties {widthOverride=500}}
  Text {
   widgetClassPath=&UEFN_TextBlock
   bIsVariable=true
   parent=SizeBox_88
   widget=Text
   properties {
    text=Text Block
    font {
     bForceMonospaced=false
     bMaterialIsStencil=false
     fontMaterial=null
     fontObject=&BurbankSmall
     letterSpacing=5
     monospacedWidth=1
     size=12
     skewAmount=0
     typefaceFontName=Black
     outlineSettings {
      bApplyOutlineToDropShadows=false
      bMiteredCorners=false
      bSeparateFillAlpha=false
      outlineMaterial=null
      outlineSize=2
      outlineColor=<rgba 0 0 0 1>
     }
    }
   }
   slot {class=SizeBoxSlot}
  }
 ]
}
