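# Debug list entry widget for the Music Event subsystem that displays the live status of one MusicClockComponent.
# It stores the clock from the list item and supplies text and color to its base entry widget.
# The text shows name, tempo, time signature, state, section, bar, beat and seconds at video render time.
# Color is green for the debug widget's highest-priority running clock, white for others running, red otherwise.
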WidgetBlueprint /Game/Sounds/MusicEventSubsystem/B_MusicEventClockWidget
graphs [
 EventGraph {nodes [
   Event Tick()
   > Parent: Tick!(execute, MyGeometry:Geometry Structure, InDeltaTime:float=0) -> (then) {is=title K2Node_CallParentFunction}
   n1 = Event OnListItemObjectSet()
   > n2 = Cast To MusicClockComponent!(execute, Object=@n1.ListItemObject) -> (then, CastFailed, AsMusic Clock:Music Clock Object Reference) {is=title K2Node_DynamicCast}
   > Set MusicClock(MusicClock:Music Clock Object Reference=@n2) -> (Output_Get:Music Clock Object Reference)
  ]}
 GetFormatText {nodes [
   GetFormatText() {is=title K2Node_FunctionEntry}
   > Get!(execute:Exec) -> (then:Exec, else:Exec, MusicClock:Music Clock Object Reference)
   n1 = Get MusicClock() -> (MusicClock:Music Clock Object Reference)
   n2 = Break MIDI Song Position!(MidiSongPos:"MIDI Song Position Structure (by ref)"=GetSongPos!(self:Music Clock Object Reference=@n1, Timebase:ECalibratedMusicTimebase Enum=VideoRenderTime) -> (ReturnValue:"MIDI Song Position Structure (by ref)")) -> (
    SecondsIncludingCountIn:float=~
    TimeSigNumerator:int
    TimeSigDenominator:int
    Tempo:float=~
    BarsIncludingCountIn:float=~
    BeatsIncludingCountIn:float=~
   ) {is=title K2Node_BreakStruct}
   n3 = GetCurrentSectionName!(self:Music Clock Object Reference=@n1, Timebase:ECalibratedMusicTimebase Enum=VideoRenderTime) -> (ReturnValue:String)
   n4 = IsEmpty(InString=@n3)
   n5 = Append!(A="Section=", B=@n3, C:String=" ") -> (ReturnValue)
   n6 = GetState!(self:Music Clock Object Reference=@n1) -> (ReturnValue:EMusicClockState Enum=Stopped)
   > Return Node!(
    execute
    Output:Text=Format Text!(
     Format="LOCTEXT(\"{Name}  [{Tempo} BPM {SigNum}/{SigDenom}] State={State} {Section}Bar={Bar} Beat={Beat} Seconds={Seconds}\")"
     Name:String=GetDisplayName(Object=@n1)
     Tempo:float=@n2.Tempo
     SigNum:int=@n2.TimeSigNumerator
     SigDenom:int=@n2.TimeSigDenominator
     State:String=Enum to String!(Enumerator:EMusicClockState Enum=@n6) -> (ReturnValue:String)
     Section:String=SelectString(B=@n5, bPickA=@n4)
     Bar:float=@n2.BarsIncludingCountIn
     Beat:float=@n2.BeatsIncludingCountIn
     Seconds:float=@n2
    ) -> (Result)
   )
  ]}
 GetTextColor {nodes [
   GetTextColor() {is=title K2Node_FunctionEntry}
   > n1 = Get()
   n2 = Get DebugWidget() -> (DebugWidget:B Music Event Debug Widget Object Reference)
   > n3 = Branch(Condition="Equal (Object)"(
     A=@n1
     B=Get HighestPriRunningClock!(self:B Music Event Debug Widget Object Reference=@n2) -> (HighestPriRunningClock:Music Clock Object Reference)
    ))
   > Return Node(Output:Slate Color Structure=Make Slate Color(SpecifiedColor=(R=0,G=1,B=0.0718,A=1)))
   n4 = Branch(
    execute=@n3.else
    Condition="Equal (Enum)"(A:EMusicClockState Enum=GetState(self=@n1) -> (ReturnValue=Stopped), B:EMusicClockState Enum=Running)
   )
   > Return Node(Output=Make Slate Color(SpecifiedColor=(R=1,G=1,B=1,A=1)))
   Return Node(execute=[@n4.else, @n1.else], Output=Make Slate Color(SpecifiedColor=(R=1,G=0.0294,B=0.0176,A=1)))
  ]}
]
variables [MusicClock {type=MusicClockComponent, default=null}]
widgets {
 info {
  inheritedWidgetCount=0
  parentClass=&/B_MusicEventBaseListEntryWidget.B_MusicEventBaseListEntryWidget_C
  rootWidgetClass=null
  widgetCount=0
 }
 widgets []
}
