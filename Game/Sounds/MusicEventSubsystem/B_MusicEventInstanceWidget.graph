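# List entry widget that shows a debug readout for one MusicEventInstance, extending B_MusicEventBaseListEntryWidget.
# GetFormatText builds a line of name, asset, trimmed tag, priority, listener distance, control bus value and optional clock, caching Distance.
# GetTextColor picks green for audible, blue for preparing to be dominant, otherwise white.
# Its distance-based alpha caps Distance with Min at the range start, so it stays opaque when max distance exceeds that.
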WidgetBlueprint /Game/Sounds/MusicEventSubsystem/B_MusicEventInstanceWidget
graphs [
 EventGraph {nodes [
   Event Tick()
   > Parent: Tick!(execute, MyGeometry:Geometry Structure, InDeltaTime:float=0) -> (then) {is=title K2Node_CallParentFunction}
   n1 = Event OnListItemObjectSet()
   > n2 = Cast To MusicEventInstance!(execute, Object=@n1.ListItemObject) -> (then, CastFailed, AsMusic Event Instance:Music Event Instance Object Reference) {is=title K2Node_DynamicCast}
   > Set EventInstance(EventInstance:Music Event Instance Object Reference=@n2) -> (Output_Get:Music Event Instance Object Reference)
  ]}
 GetAssetName {nodes [
   n1 = GetAssetName!() -> (then, Target:Music Event Instance Object Reference) {is=title K2Node_FunctionEntry}
   > n2 = Is Valid(InputObject=GetComponent!(self:Music Event Instance Object Reference=@n1) -> (ReturnValue:Audio Component Object Reference))
   Return Node!(execute=@n2.Is Not Valid, ReturnValue:String="<INVALID>")
   > Return Node(ReturnValue=GetDisplayName(Object=GetAsset!(self:Music Event Instance Object Reference=@n1) -> (ReturnValue:Sound Base Object Reference)))
  ]}
 GetFormatText {nodes [
   GetFormatText() {is=title K2Node_FunctionEntry}
   > n1 = Get!(execute:Exec) -> (then:Exec, else:Exec, EventInstance:Music Event Instance Object Reference)
   > Set Distance(Distance:double=Sqrt!(A:double=GetDistSqToListener!(self:Music Event Instance Object Reference=@n1) -> (ReturnValue:float)) -> (ReturnValue:double)) -> (Output_Get:double)
   n2 = Get Distance() -> (Distance:double)
   n3 = GetDisplayName(Object=GetClock!(self:Music Event Instance Object Reference=@n1) -> (ReturnValue:Music Clock Object Reference))
   n4 = IsEmpty(InString=@n3)
   n5 = Append!(A="Clock=", B=@n3, C:String=" ") -> (ReturnValue)
   n6 = GetDebugStringFromGameplayTag!(GameplayTag:Gameplay Tag Structure=GetEventData!(self:Music Event Instance Object Reference=@n1) -> (
     ReturnValue_EventTag:Gameplay Tag Structure
     ReturnValue_ResolutionRule:EMusicEventResolutionRule Enum
     ReturnValue_FadeInTime:float=~
     ReturnValue_FadeOutTime:float=~
     ReturnValue_MaxConcurrentEvents:int
     ReturnValue_AttenuationEvalThreshold:float=~
     ReturnValue_Description:String
    )) -> (ReturnValue:String)
   > Return Node!(
    execute
    Output:Text=Format Text!(
     Format="LOCTEXT(\"{Name} Asset={Asset} Tag={Tag} Priority={Priority} Dist={Distance} CB={CB} {Clock}\")"
     Name:String=GetDisplayName(Object=@n1)
     Asset:String=GetAssetName!(Target:Music Event Instance Object Reference=@n1) -> (ReturnValue:String) {is=title K2Node_CallFunction}
     Tag:String=Split(SourceString=@n6, InStr=MusicEventSubsystem.Event., SearchDir=FromStart).RightS
     Priority:int=GetPriority!(self:Music Event Instance Object Reference=@n1) -> (ReturnValue:int=0)
     Distance:Text="To Text (Float)"(Value=@n2, MinimumIntegralDigits=1, MaximumIntegralDigits=324, MinimumFractionalDigits=0, MaximumFractionalDigits=0)
     CB:float=GetControlBusValue!(self:Music Event Instance Object Reference=@n1) -> (ReturnValue:float)
     Clock:String=SelectString(B=@n5, bPickA=@n4)
    ) -> (Result)
   )
  ]}
 GetTextColor {nodes [
   GetTextColor() {is=title K2Node_FunctionEntry}
   > n1 = Get()
   n2 = IsPreparingToBeDominant!(self:Music Event Instance Object Reference=@n1) -> (ReturnValue:bool)
   n3 = Make Slate Color(SpecifiedColor=(R=1,G=1,B=1,A=1))
   n4 = Make Slate Color(SpecifiedColor=(R=0,G=0.1808,B=1,A=1))
   n5 = Break Slate Color!(SlateColor:"Slate Color Structure (by ref)"=Select(
     Option 0:Slate Color Structure=Select(Option 0=@n3, Option 1=@n4, Index=@n2)
     Option 1:Slate Color Structure=Make Slate Color(SpecifiedColor=(R=0,G=1,B=0.0718,A=1))
     Index=IsAudible!(self:Music Event Instance Object Reference=@n1) -> (ReturnValue:bool)
    ) -> (ReturnValue:Slate Color Structure)) -> (SpecifiedColor:Linear Color Structure, ColorUseRule:ESlateColorStylingMode Enum) {is=title K2Node_BreakStruct}
   n6 = BreakColor(InColor=@n5)
   n7 = MapRangeClamped(
    Value="Min (Float)"(A:double=Get Distance(), B:double=250) -> (ReturnValue:double) {is=title K2Node_CommutativeAssociativeBinaryOperator}
    InRangeA=250
    InRangeB=GetMaxDistance!(self:Music Event Instance Object Reference=@n1) -> (ReturnValue:float)
    OutRangeA=1
    OutRangeB=0.5
   )
   > Return Node(Output:Slate Color Structure=Make Slate Color(
     SpecifiedColor=MakeColor!(R:float=@n6, G:float=@n6.G, B:float=@n6.B, A:float=@n7) -> (ReturnValue:Linear Color Structure)
     ColorUseRule=@n5.ColorUseRule
    ))
  ]}
]
variables [
 Distance {type=number}
 EventInstance {type=MusicEventInstance, default=null}
]
widgets {
 info {
  inheritedWidgetCount=0
  parentClass=&/B_MusicEventBaseListEntryWidget.B_MusicEventBaseListEntryWidget_C
  rootWidgetClass=null
  widgetCount=0
 }
 widgets []
}
