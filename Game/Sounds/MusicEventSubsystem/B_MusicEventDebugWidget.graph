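# Debug overlay widget for the Music Event subsystem, listing event instances and music clocks in two list views on a blurred panel.
# Parent-class events supply the arrays, then pass the widget itself to displayed entry widgets as a back-reference.
# Instances are first filtered by asset-name substring, removing non-matches in reverse index order.
# Custom events set the filter, stored font size and panel position via render transform.
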WidgetBlueprint /Game/Sounds/MusicEventSubsystem/B_MusicEventDebugWidget
graphs [
 EventGraph {nodes [
   n1 = SetMusicEventAssetFilter!() -> (OutputDelegate, then, Filter:String) {is=title K2Node_CustomEvent}
   > Set AssetFilter(AssetFilter:String=@n1.Filter) -> (Output_Get:String)
   n2 = SetMusicEventFontSize!() -> (OutputDelegate, then, Size:int) {is=title K2Node_CustomEvent}
   > Set FontSize(FontSize:int=@n2.Size) -> (Output_Get:int=0)
   n3 = UpdateListEntries!() -> (OutputDelegate, then, EntryWidgets:Array of User Widget Object References) {is=title K2Node_CustomEvent}
   > n4 = For Each Loop(Array:Array of User Widget Object References=@n3.EntryWidgets) -> (Array Element:User Widget Object Reference)
   > n5 = Cast To B_MusicEventBaseListEntryWidget!(execute, Object=@n4) -> (then, CastFailed, AsB Music Event Base List Entry Widget:B Music Event Base List Entry Widget Object Reference) {is=title K2Node_DynamicCast}
   > Set DebugWidget!(
    execute
    DebugWidget:B Music Event Debug Widget Object Reference=Self-Reference()
    self:B Music Event Base List Entry Widget Object Reference=@n5
   ) -> (then, Output_Get:B Music Event Debug Widget Object Reference)
   n6 = SetMusicEventPos!() -> (OutputDelegate, then, X:double=~, Y:double=~) {is=title K2Node_CustomEvent}
   > Set RenderTransform!(
    execute
    RenderTransform:Widget Transform Structure=Make Widget Transform!(
     Translation_X:double=@n6.X
     Translation_Y:double=@n6.Y
     Scale:Vector 2D Structure=(X=1,Y=1)
     Shear:Vector 2D Structure=(X=0,Y=0)
     Angle:float=0
    ) -> (WidgetTransform:Widget Transform Structure) {is=title K2Node_MakeStruct}
    self:Widget Object Reference=Get Background() -> (Background:Background Blur Object Reference)
   ) -> (then, Output_Get:Widget Transform Structure)
   n7 = Event SetEventClocks!() -> (OutputDelegate, then, Clocks:Array of Music Clock Object References, HighestPriRunningClock:Music Clock Object Reference) {is=title K2Node_Event}
   > Set Clocks(Clocks:Array of Music Clock Object References=@n7.Clocks) -> (Output_Get:Array of Music Clock Object References)
   > Set HighestPriRunningClock(HighestPriRunningClock:Music Clock Object Reference=@n7.HighestPriRunningClock) -> (Output_Get:Music Clock Object Reference)
   > Set List Items!(
    execute
    self:List View Object Reference=Get EventClocks() -> (EventClocks:Common List View Object Reference)
    InListItems:Array of Object References=Get Clocks() -> (Clocks:Array of Music Clock Object References)
   ) -> (then) {is=title K2Node_CallFunction}
   > UpdateListEntries!(
    execute
    self
    EntryWidgets:Array of User Widget Object References=GetDisplayedEntryWidgets!(self:List View Base Object Reference=Get EventClocks()) -> (ReturnValue:Array of User Widget Object References)
   ) -> (then) {is=title K2Node_CallFunction}
   n8 = Event SetEventInstances!() -> (OutputDelegate, then, Instances:Array of Music Event Instance Object References) {is=title K2Node_Event}
   > Set Instances(Instances:Array of Music Event Instance Object References=@n8.Instances) -> (Output_Get:Array of Music Event Instance Object References)
   > FilterInstances()
   > Set List Items(
    self=Get EventInstances() -> (EventInstances:Common List View Object Reference)
    InListItems=Get Instances() -> (Instances:Array of Music Event Instance Object References)
   )
   > UpdateListEntries(EntryWidgets=GetDisplayedEntryWidgets(self=Get EventInstances()))
  ]}
 FilterInstances {nodes [
   FilterInstances() {is=title K2Node_FunctionEntry}
   > Clear(TargetArray:Array of Integers=Get IndicesToRemove() -> (IndicesToRemove:Array of Integers))
   > n1 = For Each Loop(Array:Array of Music Event Instance Object References=Get Instances()) -> (Array Element:Music Event Instance Object Reference)
   n2 = Reverse for Each Loop!(Exec:Exec=@n1.Completed, Array:Array of Integers=Get IndicesToRemove()) -> (LoopBody:Exec, ArrayIndex:int, ArrayElement:int, Completed:Exec) {is=title K2Node_MacroInstance}
   > Remove Index!(execute, TargetArray:Array of Music Event Instance Object References=Get Instances(), IndexToRemove:int=@n2.ArrayElement) -> (then)
   n3 = Contains(
    SearchIn=GetAssetName!(Target:Music Event Instance Object Reference=@n1) -> (ReturnValue:String)
    Substring=Get AssetFilter() -> (AssetFilter:String)
   )
   n4 = IsEmpty(InString=Get AssetFilter())
   Branch(execute=@n1.LoopBody, Condition=AND Boolean(A=NOT Boolean(A=@n4), B=NOT Boolean(A=@n3)))
   > Add(TargetArray:Array of Integers=Get IndicesToRemove(), NewItem:"Integer (by ref)"=@n1.Array Index) {is=title K2Node_CallArrayFunction}
  ]}
 GetAssetName {nodes [
   n1 = GetAssetName!() -> (then, Target:Music Event Instance Object Reference) {is=title K2Node_FunctionEntry}
   > n2 = Is Valid(InputObject=GetComponent!(self:Music Event Instance Object Reference=@n1) -> (ReturnValue:Audio Component Object Reference))
   Return Node!(execute=@n2.Is Not Valid, ReturnValue:String="<INVALID>")
   > Return Node(ReturnValue=GetDisplayName(Object=GetAsset!(self:Music Event Instance Object Reference=@n1) -> (ReturnValue:Sound Base Object Reference)))
  ]}
]
variables [
 AssetFilter {type=string, default=""}
 Clocks {type="array<MusicClockComponent>", default []}
 FontSize {type=integer, default=10}
 HighestPriRunningClock {type=MusicClockComponent, default=null}
 Instances {type="array<MusicEventInstance>", default []}
]
widgets {
 info {
  inheritedWidgetCount=0
  parentClass=/Script/MusicEventSystem.MusicEventSubsystemDebugWidget
  rootWidgetClass=&CanvasPanel
  widgetCount=9
 }
 widgets [
  CanvasPanel_26 {widgetClassPath=&CanvasPanel, widget=CanvasPanel_26}
  Background {
   widgetClassPath=&BackgroundBlur
   bIsVariable=true
   parent=CanvasPanel_26
   widget=Background
   properties {blurStrength=10}
   slot {
    class=CanvasPanelSlot
    properties {bAutoSize=true, layoutData {alignment=<0 0>, anchors {maximum=<0 0>, minimum=<0 0>}, offsets=<ltrb 40 40 100 30>}}
   }
  }
  VerticalBox_288 {
   widgetClassPath=&VerticalBox
   parent=Background
   widget=VerticalBox_288
   slot {class=BackgroundBlurSlot, properties {padding=<ltrb 0 0 0 0>}}
  }
  UEFN_TextBlock_C_94 {
   widgetClassPath=&UEFN_TextBlock
   parent=VerticalBox_288
   widget=UEFN_TextBlock_C_94
   properties {
    text=Events
    font {
     bForceMonospaced=false
     bMaterialIsStencil=false
     fontMaterial=null
     fontObject=&BurbankSmall
     letterSpacing=0
     monospacedWidth=1
     size=12
     skewAmount=0
     typefaceFontName=Black
     outlineSettings {
      bApplyOutlineToDropShadows=false
      bMiteredCorners=false
      bSeparateFillAlpha=false
      outlineMaterial=null
      outlineSize=0
      outlineColor=<rgba 0 0 0 1>
     }
    }
   }
   slot {class=VerticalBoxSlot}
  }
  Border_259 {
   widgetClassPath=&Border
   parent=VerticalBox_288
   widget=Border_259
   properties {visibility=HitTestInvisible, brushColor=<rgba 0 0 0 0.25>, padding=<ltrb 0 0 0 0>}
   slot {class=VerticalBoxSlot}
  }
  EventInstances {
   widgetClassPath=&CommonListView
   bIsVariable=true
   parent=Border_259
   widget=EventInstances
   properties {
    allowOverscroll=false
    bEnableRightClickScrolling=false
    bEnableScrollAnimation=false
    entryWidgetClass=&/B_MusicEventInstanceWidget.B_MusicEventInstanceWidget_C
    selectionMode=None
   }
   slot {class=BorderSlot, properties {padding=<ltrb 0 0 0 0>}}
  }
  UEFN_TextBlock_C_390 {
   widgetClassPath=&UEFN_TextBlock
   parent=VerticalBox_288
   widget=UEFN_TextBlock_C_390
   properties {
    text=Clocks
    font {
     bForceMonospaced=false
     bMaterialIsStencil=false
     fontMaterial=null
     fontObject=&BurbankSmall
     letterSpacing=0
     monospacedWidth=1
     size=12
     skewAmount=0
     typefaceFontName=Black
     outlineSettings {
      bApplyOutlineToDropShadows=false
      bMiteredCorners=false
      bSeparateFillAlpha=false
      outlineMaterial=null
      outlineSize=0
      outlineColor=<rgba 0 0 0 1>
     }
    }
   }
   slot {class=VerticalBoxSlot, properties {padding=<ltrb 0 10 0 0>}}
  }
  Border_139 {
   widgetClassPath=&Border
   parent=VerticalBox_288
   widget=Border_139
   properties {brushColor=<rgba 0 0 0 0.25>, padding=<ltrb 0 0 0 0>}
   slot {class=VerticalBoxSlot}
  }
  EventClocks {
   widgetClassPath=&CommonListView
   bIsVariable=true
   parent=Border_139
   widget=EventClocks
   properties {
    allowOverscroll=false
    bEnableRightClickScrolling=false
    bEnableScrollAnimation=false
    entryWidgetClass=&/B_MusicEventClockWidget.B_MusicEventClockWidget_C
    selectionMode=None
   }
   slot {class=BorderSlot, properties {padding=<ltrb 0 0 0 0>}}
  }
 ]
}
